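# Niagara drift trail cosmetic "Luminous Sparks" for vehicles, drawing glowing trails, wheel wraps and sparks at the rear wheels.
# Two offset ribbon emitters use quality-scaled spawn rates, game-mode-specific lifetime and velocity, and fading inherited movement.
# Socket-attached wheel meshes and short-lived, gravity-pulled sprite sparks complete the effect, with per-wheel disable masks throughout.
# User painted flags and colours override default tints, while frontend preview adjusts ribbon brightness and fade.
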NiagaraSystem /VehicleCosmetics/DriftTrail/DriftTrail_LuminousSparks/FX/NS_Trail_LuminousSparks_MP
dependencies {
 usedDataInterfaces [
  &NiagaraDataInterfaceCurve
  &NiagaraDataInterfaceSpriteRendererInfo
  &NiagaraDataInterfaceVectorCurve
  &NiagaraDataInterfaceSkeletalMesh
 ]
 usedDynamicInputs []
 usedModules [
  &self:SystemSpawnScript.SystemScriptSource.SystemScriptGraph.NiagaraNodeAssignment_0.SetVariables_#1
  &self:SystemSpawnScript.SystemScriptSource.SystemScriptGraph.NiagaraNodeAssignment_1.SetVariables_#2
  &SystemState
  &FN_UnpackVehicleFlags
  /Game/Effects/Niagara/Modules/Vehicle/FN_UpdateVehicleAttribs
  /Game/Effects/Niagara/Modules/Vehicle/FN_VehicleFlags_Cheatcodes
 ]
}
emitters {
 AcidTrail_New {
  emitterUpdateScript [
   &EmitterState(Life Cycle Mode=::NewEnumerator1, Loop Delay=0.1, Loop Duration=0.1, Scalability Mode=::NewEnumerator1, UseLoopDelay=false)
   &VC_RateQualityHandler(Epic Spawn Rate=90, High Spawn Rate=60, Low Spawn Rate=30, Medium Spawn Rate=45, Use Custom Rate?=true)
   &VC_SpawnRibbon_v1(Number of Ribbons=2, Spawn Time=-1, SpawnRate=$Emitter.Spawn Rate)
  ]
  particleSpawnScript [
   &VC_PaintedVariantsOverride() {enabled=false}
   &VC_InitializeRibbon_v1(
    Color=&MakeCustomLinearColorFromBool(Bool=$User.bIsPainted, True Color=$User.PaintedColor, False Color=<rgba 1 1 1 1>)
    Custom Offset=true
    Enable Skeletal Position=true
    Facing Axis:ENiagaraOrientationAxis=::ZAxis
    Lifetime=/VehicleCosmetics/FX/Niagara/Modules/VC_Mask_Floats(Coro Value:NiagaraFloat=0.4, FrontEnd Value:NiagaraFloat=0.5, FN Value:NiagaraFloat=0.5)
    Local Vector Offset=&AddVectorToPosition(Position=<-10 0 -6>, Vector=$User.offset)
    Offset Amount=1
    Ribbon Width=150
    Write Material Random=false
   )
   &AddVelocity(Velocity=&VC_Mask_Vectors(Coro Value:Vector3f=<-800 0 0>, FrontEnd Value:Vector3f=<-1800 0 0>, FN Value:Vector3f=<-1300 0 0>)) {module=AddVelocity001}
   /SetVariables(Particles.PinkFacing:Vector3f=<0 1 0>) {isSetParameters=true, module=SetVariables_#3}
   /VehicleCosmetics/FX/Niagara/Modules/MergeEmitters/VC_ExtraBindingColors(
    ExtraColors=1
    R1 Color:LinearColor=&MakeCustomLinearColorFromBool(Bool=$User.bIsPainted, True Color=$User.PaintedColor_3, False Color=<rgba 1 1 1 1>)
   )
   /VehicleCosmetics/FX/VC_FXParents/Modules/DLR_ProductPreviews(GameModeID:NiagaraInt32=$System.GameModeID, bPreviewVehicleSpeed:NiagaraBool=$System.bPreviewVehicleSpeed)
  ]
  particleUpdateScript [
   &FN_DisableTireFX(Mask Index Attribute:NiagaraInt32=$Particles.SpawnGroup, Mask Index FL=2, Mask Index FR=3, Mask Index RL=0, Mask Index RR=1)
   &InheritSourceMovement(
    Applied Position Scale=&Multiply_VectorByFloat(
     Vector=<1 1 1>
     Float=&FloatFromCurve(FloatCurve=DataInterface() {propertyValues {Curve {keys [
          {arriveTangent=0.0304, interpMode=RCIM_Cubic, leaveTangent=-0.008, tangentMode=RCTM_Break, time=0, value=1}
          {arriveTangent=-0.0077, interpMode=RCIM_Cubic, leaveTangent=0.0277, tangentMode=RCTM_Break, time=1, value=0.9}
         ]}}})
    )
    ApplyForce=false
    ApplyPosition=true
    ApplyPositionFalloff=false
   ) {enabled=false, module=InheritSourceMovement001}
   &ParticleState()
   &DynamicMaterialParameters(
    Index 0 Param 1=0.25
    Index 0 Param 2=0.25
    Index 0 Param 4=&Waveform("Waveform [1]"=::NewEnumerator0, "[1] Frequency"=0.7, "[1] Amplitude Min/Max"=<3 5>)
    Index 1 Param 1=&RandomRangeFloat(Minimum=20, Maximum=30)
    Index 1 Param 2=3
    Index 1 Param 3=-30
    Index 1 Param 4=0.3
    Index 2 Param 1=-33
    Module.Index0Param0Write=true
    Module.Index0Param1Write=true
    Module.Index0Param2Write=true
    Module.Index0Param3Write=true
    Module.Index1Param0Write=true
    Module.Index1Param1Write=true
    Module.Index1Param2Write=true
    Module.Index1Param3Write=true
    Module.Index2Param0Write=true
    Module.Index2Param1Write=true
    Module.Index2Param2Write=true
    Module.Index2Param3Write=true
    Write Parameter Index 1=true
    Write Parameter Index 2=true
   ) {module=DynamicMaterialParameters001}
   &InheritSourceMovement(
    Applied Position Falloff Distance=650
    Applied Position Scale=&VectorFromFloat(Value=&FloatFromCurve(FloatCurve=DataInterface() {propertyValues {Curve {
         defaultValue=1
         keys [
          {arriveTangent=-1.7885, interpMode=RCIM_Cubic, leaveTangent=-1.7885, tangentMode=RCTM_User, time=0.05, value=1}
          {interpMode=RCIM_Cubic, tangentMode=RCTM_User, time=1, value=0}
         ]
        }}}))
    ApplyForce=false
    ApplyPosition=true
    ApplyPositionFalloff=false
   )
   &ScaleColor(
    Scale Alpha=&FloatFromCurve(FloatCurve=DataInterface() {propertyValues {Curve {keys [
         <at 0 0>
         <at 0.05 1>
         {interpMode=RCIM_Cubic, tangentMode=RCTM_Break, time=0.9, value=1}
         {interpMode=RCIM_Cubic, tangentMode=RCTM_Break, time=1, value=0}
        ]}}})
    Scale RGB=&MakeCustomVectorFromBool(
     Bool=$User.bIsPainted
     True Vector=&MakeCustomVectorFromBool(Bool=$User.bIsFrontendPreview, True Vector=<1 1 1>, False Vector=<0.15 0.15 0.15>)
     False Vector=&MakeCustomVectorFromBool(
      Bool=$User.bIsFrontendPreview
      True Vector=&VectorFromCurve(
       VectorCurve=DataInterface() {propertyValues {
         XCurve {keys [
           {interpMode=RCIM_Cubic, leaveTangentWeight=0.3333, tangentMode=RCTM_Break, time=0, value=2}
           {tangentMode=RCTM_User, time=0.3, value=0.2}
           <at 1 0>
          ]}
         YCurve {keys [
           {interpMode=RCIM_Cubic, leaveTangentWeight=0.3333, tangentMode=RCTM_Break, time=0, value=0.15}
           {tangentMode=RCTM_User, time=0.3, value=1}
           <at 1 0>
          ]}
         ZCurve {keys [
           {interpMode=RCIM_Cubic, leaveTangentWeight=0.3333, tangentMode=RCTM_Break, time=0, value=0.3}
           {tangentMode=RCTM_User, time=0.3, value=1}
           <at 1 0>
          ]}
        }}
       Scale Curve=<0.5 1 1>
      )
      False Vector=&VectorFromCurve(VectorCurve=DataInterface() {propertyValues {
         XCurve {keys [
           {interpMode=RCIM_Cubic, leaveTangentWeight=0.3333, tangentMode=RCTM_Break, time=0, value=1}
           {arriveTangent=-0.6262, tangentMode=RCTM_User, time=0.3, value=0.2}
           {arriveTangent=-0.2299, interpMode=RCIM_Cubic, tangentMode=RCTM_Break, time=1, value=0}
          ]}
         YCurve {keys [
           {interpMode=RCIM_Cubic, leaveTangentWeight=0.3333, tangentMode=RCTM_Break, time=0, value=0.15}
           {tangentMode=RCTM_User, time=0.3, value=1}
           <at 1 0>
          ]}
         ZCurve {keys [
           {interpMode=RCIM_Cubic, leaveTangentWeight=0.3333, tangentMode=RCTM_Break, time=0, value=0.3}
           {interpMode=RCIM_Cubic, tangentMode=RCTM_Break, time=0.3, value=2}
           {arriveTangent=-3.6591, interpMode=RCIM_Cubic, tangentMode=RCTM_Break, time=1, value=0}
          ]}
        }})
     )
    )
    ScaleA=true
    ScaleRGB=true
   ) {module=ScaleColor002}
   &SolveForcesAndVelocity(Clamp Velocity=false, Limit Acceleration=false)
   &VC_UpdateRibbon_v1(Detach Locked Particle=true, Enable Skeletal Position=true, Locked Particle Detach Time=-1)
  ]
  rendererClasses [&NiagaraRibbonRendererProperties]
  renderers [{rendererClass=&NiagaraRibbonRendererProperties}]
 }
 AcidTrail_New002 {
  emitterSpawnScript [/SetVariables(Emitter.BlueRibbonWidthM=150, Emitter.PinkRibbonWidthM=50) {isSetParameters=true}]
  emitterUpdateScript [
   &EmitterState(Life Cycle Mode=::NewEnumerator1, Loop Delay=0.1, Loop Duration=0.1, Scalability Mode=::NewEnumerator1, UseLoopDelay=false)
   &VC_RateQualityHandler(Epic Spawn Rate=90, High Spawn Rate=60, Low Spawn Rate=30, Medium Spawn Rate=45, Use Custom Rate?=true)
   &VC_SpawnRibbon_v1(Number of Ribbons=2, Spawn Time=-1, SpawnRate=$Emitter.Spawn Rate)
  ]
  particleSpawnScript [
   &VC_PaintedVariantsOverride() {enabled=false}
   &VC_InitializeRibbon_v1(
    Custom Offset=true
    Enable Skeletal Position=true
    Facing Axis:ENiagaraOrientationAxis=::ZAxis
    Lifetime=&VC_Mask_Floats(Coro Value:NiagaraFloat=0.4, FrontEnd Value:NiagaraFloat=0.5, FN Value:NiagaraFloat=0.5)
    Local Vector Offset=&AddVectorToPosition(Position=<-10 0 -36>, Vector=$User.offset)
    Offset Amount=1
    Ribbon Width=150
    Write Material Random=false
   )
   &AddVelocity(Velocity=&VC_Mask_Vectors(Coro Value:Vector3f=<-800 0 0>, FrontEnd Value:Vector3f=<-1800 0 0>, FN Value:Vector3f=<-1300 0 0>)) {module=AddVelocity001}
   /SetVariables(Particles.PinkFacing:Vector3f=<0 1 0>) {isSetParameters=true, module=SetVariables_#3}
   &VC_ExtraBindingColors(
    ExtraColors=1
    R1 Color:LinearColor=&MakeCustomLinearColorFromBool(Bool=$User.bIsPainted, True Color=$User.PaintedColor_3, False Color=<rgba 1 1 1 1>)
   )
   &DLR_ProductPreviews(GameModeID:NiagaraInt32=$System.GameModeID, bPreviewVehicleSpeed:NiagaraBool=$System.bPreviewVehicleSpeed)
  ]
  particleUpdateScript [
   &FN_DisableTireFX(Mask Index Attribute:NiagaraInt32=$Particles.SpawnGroup, Mask Index FL=2, Mask Index FR=3, Mask Index RL=0, Mask Index RR=1)
   &InheritSourceMovement(
    Applied Position Scale=&Multiply_VectorByFloat(
     Vector=<1 1 1>
     Float=&FloatFromCurve(FloatCurve=DataInterface() {propertyValues {Curve {keys [
          {arriveTangent=0.0304, interpMode=RCIM_Cubic, leaveTangent=-0.008, tangentMode=RCTM_Break, time=0, value=1}
          {arriveTangent=-0.0077, interpMode=RCIM_Cubic, leaveTangent=0.0277, tangentMode=RCTM_Break, time=1, value=0.9}
         ]}}})
    )
    ApplyForce=false
    ApplyPosition=true
    ApplyPositionFalloff=false
   ) {enabled=false, module=InheritSourceMovement001}
   &ParticleState()
   &InheritSourceMovement(
    Applied Position Falloff Distance=650
    Applied Position Scale=&VectorFromFloat(Value=&FloatFromCurve(FloatCurve=DataInterface() {propertyValues {Curve {
         defaultValue=1
         keys [
          {arriveTangent=-1.7885, interpMode=RCIM_Cubic, leaveTangent=-1.7885, tangentMode=RCTM_User, time=0.05, value=1}
          {interpMode=RCIM_Cubic, tangentMode=RCTM_User, time=1, value=0}
         ]
        }}}))
    ApplyForce=false
    ApplyPosition=true
    ApplyPositionFalloff=false
   )
   /SetVariables(
    Particles.PinkRibbonWidth:NiagaraFloat=&Multiply_Float(
     A=$Emitter.PinkRibbonWidthM
     B=&FloatFromCurve(FloatCurve=DataInterface() {propertyValues {Curve {keys [<at 0 0>,<at 0.01 0>,<at 0.065 0>,<at 0.085 1>,<at 1 1>]}}})
    )
    Particles.R1 Color:LinearColor=&VC_DynamicScaleColor(
     Color to Scale:LinearColor=$Particles.Initial.R1 Color
     RGB Scale:Vector3f=&MakeCustomVectorFromBool(
      Bool=$User.bIsPainted
      True Vector=<1 1 1>
      False Vector=&VectorFromCurve(
       VectorCurve=DataInterface() {propertyValues {
         XCurve {keys [<at 0 2>,<at 0.2805 0>,<at 1 0>]}
         YCurve {keys [<at 0 0.15>,<at 0.2805 0.6822>,<at 1 0>]}
         ZCurve {keys [<at 0 1>,<at 0.2805 1>,<at 1 0>]}
        }}
       Scale Curve=<1.5 1.5 1.5>
      )
     )
     Alpha Scale:NiagaraFloat=&MakeCustomFloatFromBool(
      Bool=$User.bIsFrontendPreview
      True Float=&Multiply_Float(
       A=1
       B=&FloatFromCurve(FloatCurve=DataInterface() {propertyValues {Curve {keys [<at 0 0>,<at 0.2 1>,<at 1 1>]}}}, Scale Curve=2.5)
      )
      False Float=&FloatFromCurve(FloatCurve=DataInterface() {propertyValues {Curve {keys [<at 0 0>,<at 0.2 1>,<at 0.8 1>,<at 1 0>]}}}, Scale Curve=2.5)
     )
    )
   ) {isSetParameters=true}
   &ScaleColor(
    Scale Alpha=&FloatFromCurve(FloatCurve=DataInterface() {propertyValues {Curve {keys [
         <at 0 0>
         <at 0.05 1>
         {interpMode=RCIM_Cubic, tangentMode=RCTM_Break, time=0.9, value=1}
         {interpMode=RCIM_Cubic, tangentMode=RCTM_Break, time=1, value=0}
        ]}}})
    Scale RGB=&MakeCustomVectorFromBool(
     Bool=$User.bIsPainted
     True Vector=&MakeCustomVectorFromBool(Bool=$User.bIsFrontendPreview, True Vector=<1 1 1>, False Vector=<0.15 0.15 0.15>)
     False Vector=&MakeCustomVectorFromBool(
      Bool=$User.bIsFrontendPreview
      True Vector=&VectorFromCurve(
       VectorCurve=DataInterface() {propertyValues {
         XCurve {keys [
           {interpMode=RCIM_Cubic, leaveTangentWeight=0.3333, tangentMode=RCTM_Break, time=0, value=2}
           {tangentMode=RCTM_User, time=0.3, value=0.2}
           <at 1 0>
          ]}
         YCurve {keys [
           {interpMode=RCIM_Cubic, leaveTangentWeight=0.3333, tangentMode=RCTM_Break, time=0, value=0.15}
           {tangentMode=RCTM_User, time=0.3, value=1}
           <at 1 0>
          ]}
         ZCurve {keys [
           {interpMode=RCIM_Cubic, leaveTangentWeight=0.3333, tangentMode=RCTM_Break, time=0, value=0.3}
           {tangentMode=RCTM_User, time=0.3, value=1}
           <at 1 0>
          ]}
        }}
       Scale Curve=<0.5 1 1>
      )
      False Vector=&VectorFromCurve(VectorCurve=DataInterface() {propertyValues {
         XCurve {keys [
           {interpMode=RCIM_Cubic, leaveTangentWeight=0.3333, tangentMode=RCTM_Break, time=0, value=1}
           {arriveTangent=-0.6262, tangentMode=RCTM_User, time=0.3, value=0.2}
           {arriveTangent=-0.2299, interpMode=RCIM_Cubic, tangentMode=RCTM_Break, time=1, value=0}
          ]}
         YCurve {keys [
           {interpMode=RCIM_Cubic, leaveTangentWeight=0.3333, tangentMode=RCTM_Break, time=0, value=0.15}
           {tangentMode=RCTM_User, time=0.3, value=1}
           <at 1 0>
          ]}
         ZCurve {keys [
           {interpMode=RCIM_Cubic, leaveTangentWeight=0.3333, tangentMode=RCTM_Break, time=0, value=0.3}
           {interpMode=RCIM_Cubic, tangentMode=RCTM_Break, time=0.3, value=2}
           {arriveTangent=-3.6591, interpMode=RCIM_Cubic, tangentMode=RCTM_Break, time=1, value=0}
          ]}
        }})
     )
    )
    ScaleA=true
    ScaleRGB=true
   ) {module=ScaleColor002}
   &SolveForcesAndVelocity(Clamp Velocity=false, Limit Acceleration=false)
   &VC_UpdateRibbon_v1(Detach Locked Particle=true, Enable Skeletal Position=true, Locked Particle Detach Time=-1)
  ]
  rendererClasses [&NiagaraRibbonRendererProperties]
  renderers [{rendererClass=&NiagaraRibbonRendererProperties}]
 }
 NE_VC_DriftTrail_WheelWrap_Parent1 {
  emitterUpdateScript [
   &EmitterState(Life Cycle Mode=::NewEnumerator1, Loop Behavior=::NewEnumerator1, Loop Duration Mode=::NewEnumerator1, UseLoopDelay=false)
   &SpawnBurst_Instantaneous(Spawn Count=2, Use Spawn Probability=false, UseLoopCountLimit=false)
  ]
  particleSpawnScript [
   &VC_PaintedVariantsOverride() {enabled=false}
   &InitializeParticle(
    Color Mode=::NewEnumerator0
    Lifetime=$Emitter.CurrentLoopDuration
    Mesh Scale Mode=::NewEnumerator4
    UsePositionOffset=false
    Write Material Random=true
   )
   &PartitionParticles()
   &SkeletalMeshLocation(
    Bone / Socket Index=&Add_Integer(A=$Particles.PartitionParticles.Partition, B=4)
    Bone / Socket Sampling Mode=::NewEnumerator5
    Interpolate Mesh Sampling=false
    MaskSpawnGroup=false
    Mesh Sampling Type=::NewEnumerator6
    Skeletal Mesh=$System.SkeletalMesh
   )
   &Color(Color=&MakeCustomLinearColorFromBool(Bool=$User.bIsPainted, True Color=$User.PaintedColor_4, False Color=<rgba 0.75 0.3047 0.6566 1>))
  ]
  particleUpdateScript [
   &FN_DisableTireFX(
    Mask Index Attribute:NiagaraInt32=$Particles.PartitionParticles.Partition
    Mask Index FL=2
    Mask Index FR=3
    Mask Index RL=0
    Mask Index RR=1
   )
   &UpdateMeshOrientation(Rotation Coordinate Space=::NewEnumerator2, Rotation Rate=3, Rotation Vector=<0 1 0>) {enabled=false, module=UpdateMeshOrientation001}
   &VC_WheelWrapScale(Wheel Index Attribute:NiagaraInt32=&Add_Integer(A=$Particles.PartitionParticles.Partition, B=2))
   &DynamicMaterialParameters(
    Index 0 Param 1=&RandomRangeFloat(Minimum=0, Maximum=5)
    Module.Index0Param0Write=true
    Module.Index0Param1Write=true
    Module.Index0Param2Write=true
    Module.Index0Param3Write=true
   ) {enabled=false}
   &ParticleState(Kill Particles When Lifetime Has Elapsed=false, Loop Particles Lifetime=true)
   &ScaleColor(
    Scale RGB=&Multiply_VectorByFloat(
     Vector=&MakeCustomVectorFromBool(Bool=$User.bIsFrontendPreview, True Vector=<1 1 1>, False Vector=<1 1 1>)
     Float=4
    )
    ScaleA=true
    ScaleRGB=true
   )
   &DynamicMaterialParameters(
    Index 0 Param 1=&RandomRangeFloat(Minimum=1, Maximum=1)
    Index 0 Param 2=&RandomRangeFloat(Minimum=0, Maximum=200)
    Module.Index0Param0Write=true
    Module.Index0Param1Write=true
    Module.Index0Param2Write=true
    Module.Index0Param3Write=true
   ) {enabled=false, module=DynamicMaterialParameters001}
   &UpdateMeshOrientation(Rotation Coordinate Space=::NewEnumerator2, Rotation Vector=<0 1 0>) {enabled=false}
   &SkeletalMeshLocation(
    Bone / Socket Index=&Add_Integer(A=$Particles.PartitionParticles.Partition, B=4)
    Bone / Socket Sampling Mode=::NewEnumerator5
    MaskSpawnGroup=false
    Mesh Sampling Type=::NewEnumerator6
    Skeletal Mesh=$System.SkeletalMesh
   ) {module=SkeletalMeshLocation001}
   &SolveForcesAndVelocity(Clamp Velocity=false, Limit Acceleration=false)
  ]
  rendererClasses [&NiagaraMeshRendererProperties]
  renderers [{rendererClass=&NiagaraMeshRendererProperties, rendererIndex=0}]
 }
 Sparks_New {
  emitterUpdateScript [
   &EmitterState()
   &SpawnRate(SpawnRate=&RandomRangeFloat(Minimum=35, Maximum=40), Use Spawn Probability=false)
  ]
  particleSpawnScript [
   &VC_PaintedVariantsOverride(
    Black=<rgba 0.0938 0.0938 0.0938 1.5>
    BurntSienna=<rgba 0.9531 0.5371 0 2>
    Cobalt=<rgba 0.4792 0.4888 1 1.5>
    Crimson=<rgba 0.4627 0 0.0118 1.5>
    ForestGreen=<rgba 0 0.502 0.0235 1.5>
    Gold:LinearColor=<rgba 0.5961 0.3961 0.0471 1>
    Grey=<rgba 0.2039 0.2039 0.2039 1.5>
    Lime=<rgba 0.3776 0.7552 0 1.5>
    Orange=<rgba 0.8698 0.4017 0 1.5>
    Pink=<rgba 0.8872 0.417 0.9531 4>
    Purple=<rgba 0.2789 0 0.6795 1.5>
    Saffron=<rgba 0.9 0.69 0 1.5>
    SkyBlue=<rgba 0 0.357 0.5365 1.5>
    TitaniumWhite=<rgba 1 1 1 1.5>
   )
   &InitializeParticle(
    Color Mode=::NewEnumerator0
    Lifetime=&RandomRangeFloat(Minimum=0.175, Maximum=0.4)
    Position Offset=&AddVectorToPosition(Position=<0 0 0>, Vector=$User.offset)
    Position Offset Coordinate Space=::Local
    Sprite Size Max=&Multiply_Vector2D_ByFloat(Float=3, Vector2D=<1 44>)
    Sprite Size Min=&Multiply_Vector2D_ByFloat(Float=3, Vector2D=<1 60>)
    Sprite Size Mode=::NewEnumerator2
    UsePositionOffset=true
    Write Material Random=true
   )
   &PartitionParticles()
   &SkeletalMeshLocation(
    Bone / Socket Index=&Add_Integer(A=$Particles.PartitionParticles.Partition, B=4)
    Bone / Socket Sampling Mode=::NewEnumerator5
    Interpolate Mesh Sampling=false
    MaskSpawnGroup=false
    Mesh Sampling Type=::NewEnumerator6
    Sampled Position Coordinate Space=::Local
    Sampled Position Offset=&AddVectorToPosition(Position=<-20 0 -45>, Vector=$User.offset)
    Skeletal Mesh=$System.SkeletalMesh
   )
   &AddVelocity(Velocity=&RandomRangeVector(Minimum=<-600 -400 400>, Maximum=<-400 400 100>))
   &Color(
    Color=&MakeCustomLinearColorFromBool(Bool=$User.bIsPainted, True Color=$Particles.Color, False Color=<rgba 2 0.35 0.1 1>)
    Scale Alpha=&RandomRangeFloat(Minimum=0.001, Maximum=0.03)
   )
  ]
  particleUpdateScript [
   &FN_DisableTireFX(
    Mask Index Attribute:NiagaraInt32=$Particles.PartitionParticles.Partition
    Mask Index FL=2
    Mask Index FR=3
    Mask Index RL=0
    Mask Index RR=1
   )
   &ParticleState()
   &ScaleSpriteSize(Uniform Curve Sprite Scale=DataInterface() {propertyValues {Curve {keys [
        {interpMode=RCIM_Cubic, tangentMode=RCTM_Break, time=0, value=0}
        {interpMode=RCIM_Cubic, tangentMode=RCTM_Break, time=0.1, value=1}
        {interpMode=RCIM_Cubic, tangentMode=RCTM_Break, time=1, value=0}
       ]}}})
   &GravityForce(Gravity=&RandomRangeVector(Minimum=<0 0 -1400>, Maximum=<0 0 -980>))
   &ScaleColor(
    Scale RGB=&Multiply_VectorByFloat(
     Vector=<1 1 1>
     Float=&FloatFromCurve(FloatCurve=DataInterface() {propertyValues {Curve {keys [<at 0 512>,<at 0.2 32>]}}}, Scale Curve=2)
    )
    ScaleA=true
    ScaleRGB=true
   )
   &SubUVAnimation(
    End Frame Range Override=3
    "Sprite Renderer "=DataInterface() {propertyValues {SpriteRenderer=&self:Sparks_New_0.NiagaraSpriteRendererProperties_0}}
    SubUV Animation Mode:ENiagara_SubUVLookupModeV2=::NewEnumerator2
    UseEndFrame=true
    UseStartFrame=true
   )
   &SolveForcesAndVelocity(Clamp Velocity=false, Limit Acceleration=false)
  ]
  rendererClasses [&NiagaraSpriteRendererProperties]
  renderers [{rendererClass=&NiagaraSpriteRendererProperties}]
 }
}
summary {
 systemName=NS_Trail_LuminousSparks_MP
 emitters [
  {emitterName=Sparks_New, rendererClasses [&NiagaraSpriteRendererProperties]}
  {emitterName=NE_VC_DriftTrail_WheelWrap_Parent1, rendererClasses [&NiagaraMeshRendererProperties]}
  {emitterName=AcidTrail_New, rendererClasses [&NiagaraRibbonRendererProperties]}
  {emitterName=AcidTrail_New002, rendererClasses [&NiagaraRibbonRendererProperties]}
 ]
 userVariables [
  User.bIsFrontendPreview {type=NiagaraBool}
  User.bIsPainted {type=NiagaraBool}
  User.bPreviewVehicleSpeed {type=NiagaraBool}
  User.ColorVariantID {type=NiagaraInt32}
  User.GameModeID {type=NiagaraInt32, defaultValue=3}
  User.offset {type=Vector3f, description="offset from wheels(set through cosmetic template)", defaultValue {struct=Vector3f, value=<0 0 0>}}
  User.PaintedColor {type=LinearColor, description="", defaultValue {struct=LinearColor, value=<rgba 1 1 1 1>}}
  User.PaintedColor_2 {type=LinearColor, description="", defaultValue {struct=LinearColor, value=<rgba 1 1 1 1>}}
  User.PaintedColor_3 {type=LinearColor, description="", defaultValue {struct=LinearColor, value=<rgba 1 1 1 1>}}
  User.PaintedColor_4 {type=LinearColor, description="", defaultValue {struct=LinearColor, value=<rgba 1 1 1 1>}}
  User.PaintedScalar {type=NiagaraFloat, defaultValue=1, description=""}
  User.VehicleFlags {type=NiagaraInt32, defaultValue=806289408}
  User.Wheel_Radius_Array {
   type=NiagaraDataInterfaceArrayFloat
   defaultValue {
    struct=NiagaraExt_VariableValue_DataInterface
    value {
     dataInterface=&self:NiagaraDataInterfaceArrayFloat_0
     dataInterfaceClass=&NiagaraDataInterfaceArrayFloat
    }
   }
  }
  User.Wheel_Width_Array {
   type=NiagaraDataInterfaceArrayFloat
   defaultValue {
    struct=NiagaraExt_VariableValue_DataInterface
    value {
     dataInterface=&self:NiagaraDataInterfaceArrayFloat_1
     dataInterfaceClass=&NiagaraDataInterfaceArrayFloat
    }
   }
  }
  User.WheelSize {type=NiagaraFloat, defaultValue=1, description=""}
 ]
}
