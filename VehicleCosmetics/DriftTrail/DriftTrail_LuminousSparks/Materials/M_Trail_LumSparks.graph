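# Translucent, unlit Niagara ribbon material for the Luminous Sparks vehicle drift trail, drawing a glowing streak with animated sparks.
# Panned, rotated noise and spark textures, distorted by a flow map with sine-pulsed strength, are shaped by gradient and mask textures.
# Particle color tints sparks and glow, with dynamic parameters controlling glow brightness and an optional noise-driven hue shift.
# Opacity combines both layers with particle alpha, a U-axis fade and depth fade.
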Material /VehicleCosmetics/DriftTrail/DriftTrail_LuminousSparks/Materials/M_Trail_LumSparks
graph {
 properties {
  bUsedWithNiagaraRibbons=true
  blendMode=BLEND_Translucent
  numCustomizedUVs=1
  pixelDepthOffsetMode=PDOM_Legacy
  shadingModel=MSM_Unlit
  twoSided=true
  usedShadingModels=Unlit
 }
 nodes [
  MaterialFunctionCall(UVs=TextureCoordinate(), Rotation Center=Constant() {r=0.5}, "Rotation Angle (0-1)"=Constant() {r=-0.013}) {materialFunction=&CustomRotator}
  n1 = TextureCoordinate() {coordinateIndex=1}
  n2 = TextureSample(UVs=Panner(Coordinate=TextureCoordinate(), Time=Constant() {r=0.5}) {speedY=0.4}) {texture=/DealershipSharedLibrary/FX/Textures/TexturePackage/T_GradientMasks_Luminous_RGB}
  n3 = Multiply(A=@n2.G, B=Constant() {r=2})
  n4 = Multiply(
   A=@n3
   B=TextureSample(UVs=Panner(Coordinate=@n1) {speedX=-2}) {texture=/DealershipSharedLibrary/FX/Textures/Noise/T_VCVFX_DistortNoise02_RGB}.G
  )
  n5 = TextureCoordinate() {coordinateIndex=1, uTiling=0.5}
  n6 = Multiply(
   A=@n3
   B=TextureSample(UVs=Panner(Coordinate=@n5) {speedX=-1}) {texture=&T_VCVFX_DistortNoise02_RGB}.B
  )
  n7 = Rotator(Coordinate=TextureCoordinate() {uTiling=0.25, vTiling=0.25}, Time=Constant() {r=0.05}) {speed=3.14}
  n8 = Multiply(A=Time(), B=Constant() {r=0.25})
  n9 = ComponentMask(None=TextureSample(UVs=Panner(Coordinate=@n7, Time=@n8) {speedX=-2, speedY=1}) {texture=/DealershipSharedLibrary/FX/Textures/Noise/T_VCVFX_Flow03_N}.RGB) {g=true, r=true}
  MaterialExpressionConstantClamp_0 = Saturate(None=Sine(None=Time()) {period=2})
  n10 = Multiply(A=@n9, B=LinearInterpolate(A=Constant() {r=0.25}, B=Constant() {r=0.8}, Alpha=@MaterialExpressionConstantClamp_0))
  n11 = Panner(
   Coordinate=Rotator(Coordinate=TextureCoordinate() {vTiling=0.25}, Time=Constant() {r=0.1}) {speed=3.14}
   Time=Multiply(A=Time(), B=Constant() {r=0.75})
  ) {speedX=-2.5, speedY=0.5}
  n12 = TextureSample(UVs=Add(A=@n11, B=@n10)) {texture=&T_VCVFX_DistortNoise02_RGB}
  Multiply(A=Add(A=@n12.R, B=Add(A=@n4, B=@n6)), B=@n2.R)
  MaterialExpressionVertexColor_0 = ParticleColor()
  n13 = TextureSample() {texture=/DealershipSharedLibrary/FX/Textures/TexturePackage/T_VCVFX_Masks_RGB}
  n14 = TextureCoordinate() {vTiling=-1}
  n15 = Multiply(A=@n13.R, B=@n2.B)
  n16 = Constant() {r=4}
  n17 = TextureSample(UVs=Add(A=Multiply(A=@n16, B=@n11), B=@n10)) {texture=/DealershipSharedLibrary/FX/Textures/TexturePackage/T_VCVFX_TrailSparks_RGB}
  n18 = Multiply(
   A=Multiply(A=TextureSample(UVs=@n14) {texture=&T_VCVFX_Masks_RGB}.RGB, B=Multiply(A=@n15, B=@n13.G))
   B=Multiply(A=Constant() {r=60}, B=@n17.B)
  )
  n19 = TextureSample(UVs=Add(A=@n11, B=@n9)) {texture=&T_VCVFX_DistortNoise02_RGB}
  n20 = Multiply(A=Multiply(A=@n17.G, B=Constant() {r=20}), B=@n19.G)
  n21 = Multiply(A=@n2.R, B=@n3)
  n22 = Multiply(A=@n13.A, B=Add(A=Multiply(A=@n18, B=@n19.B), B=Multiply(A=@n20, B=@n21)))
  n23 = Add(A=Multiply(A=@n12.G, B=@n21), B=@n21)
  MaterialExpressionConstantClamp_2 = Saturate(None=Add(A=@n22, B=@n23))
  n24 = OneMinus(None=ComponentMask(None=TextureCoordinate()) {r=true})
  Multiply(
   A=MaterialFunctionCall(Opacity In=Constant() {r=1}) {materialFunction=&MF_VCVFX_DepthFade}.Result
   B=Multiply(A=Multiply(A=@MaterialExpressionVertexColor_0.A, B=@MaterialExpressionConstantClamp_2), B=Power(Base=@n24) {constExponent=1})
  )
  n25 = Multiply(A=@MaterialExpressionVertexColor_0.RGB, B=Multiply(A=Constant() {r=5}, B=@n22))
  n26 = SquareRoot(None=@MaterialExpressionVertexColor_0.RGB)
  n27 = Multiply(A=@n26, B=@n26)
  n28 = DynamicParameter() {paramNames [HueShift_Sharpness, HueShift_Amount, HueShift_Toggle, Brightness]}
  MaterialExpressionConstantClamp_1 = Saturate(None=Multiply(A=Constant() {r=0.5}, B=@n12.B))
  n29 = Multiply(A=Subtract(A=@MaterialExpressionConstantClamp_1, B=@n28.HueShift_Amount), B=@n28.HueShift_Sharpness)
  n30 = Multiply(
   A=LinearInterpolate(
    A=Multiply(A=@n27, B=@n28.Brightness)
    B=MaterialFunctionCall(Hue Shift Percentage=@n29, Texture=@MaterialExpressionVertexColor_0.RGB) {materialFunction=&HueShift}.Result
    Alpha=@n28.HueShift_Toggle
   )
   B=Multiply(A=@n23, B=@n28.Brightness)
  )
  Reroute(None=MaterialFunctionCall(Emissive=Add(A=@n25, B=@n30)) {materialFunction=&MF_FX_EmissiveAdjustments}.Result)
 ]
}
