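# Additive unlit Niagara sprite material for sparks in the Luminous Sparks vehicle drift trail.
# An ember texture's alpha, scaled by particle alpha, offsets the U coordinate of a second sample of that texture, distorting the sparks.
# A cloud noise strip with flattened V scrolls at a rate scaled by particle alpha, modulating the sparks over time.
# The product is tinted by particle color and passed through the shared emissive adjustments function.
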Material /VehicleCosmetics/DriftTrail/DriftTrail_LuminousSparks/Materials/M_Trail_Sparks
graph {
 properties {bUsedWithNiagaraSprites=true, blendMode=BLEND_Additive, pixelDepthOffsetMode=PDOM_Legacy, shadingModel=MSM_Unlit, usedShadingModels=Unlit}
 nodes [
  MaterialExpressionVertexColor_1 = ParticleColor()
  n1 = TextureCoordinate()
  n2 = ComponentMask(None=@n1) {r=true}
  n3 = Multiply(A=@MaterialExpressionVertexColor_1.A, B=TextureSample() {texture=/DealershipSharedLibrary/FX/Textures/Gradient_Masks/T_VCVFX_Embers_A}.A)
  n4 = TextureSample(UVs=AppendVector(A=Add(A=@n2, B=@n3), B=ComponentMask(None=@n1) {g=true})) {texture=&T_VCVFX_Embers_A}
  n5 = Multiply(A=@MaterialExpressionVertexColor_1.A, B=Time())
  n6 = Constant() {r=25}
  n7 = TextureSample(UVs=Panner(Coordinate=TextureCoordinate() {uTiling=0.4, vTiling=0}, Time=Multiply(A=@n5, B=@n6)) {speedY=1}) {texture=/VehicleCosmetics/FX/Textures/Noise/T_VCVFX_Clouds_A}
  MaterialFunctionCall(Emissive=Multiply(A=@MaterialExpressionVertexColor_1.RGB, B=Multiply(A=@n4.RGB, B=@n7.R))) {materialFunction=&MF_FX_EmissiveAdjustments}
 ]
}
